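# Physics asset for the Eternal Vanguard backpack cosmetic, defining secondary motion for its hoses, straps, arm and panels.
# Each bone has one capsule body, with the pack fully angularly locked to root as the anchor.
# Two short hose chains and two straps hang off the pack with tight swing limits.
# An arm-to-panel chain bends mostly on one swing axis, with rigid joints at the middle arm segment and cable.
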PhysicsAsset /BRCosmetics/Accessories/FORT_Backpacks/Backpack_F_MED_Eternal_Vanguard/Meshes/F_MED_Eternal_Vanguard_Pack_Physics
bodies {
 dyn_arm1 [{shapeType=Capsule, length=9, radius=1, shapeName=dyn_arm1_capsule, center=<5.168 0.0112 -0.3264>, rotation=<rot 0 90.1544 -88.9902>}]
 dyn_arm2 [{shapeType=Capsule, length=3, radius=1, shapeName=dyn_arm2_capsule, center=<1.4762 0.0469 0>, rotation=<rot 90 0 0>}]
 dyn_arm3 [{shapeType=Capsule, length=8.5, radius=1, shapeName=dyn_arm3_capsule, center=<6.9892 0.0344 0.1029>, rotation=<rot 0 90.281 -88.6575>}]
 dyn_cable [{shapeType=Capsule, length=0.5, radius=0.5, shapeName=dyn_cable_capsule}]
 dyn_hose1 [{shapeType=Capsule, length=8, radius=1.6, shapeName=dyn_hose1_capsule, center=<3.9249 -0.9865 -3.074>, rotation=<rot 70.4666 3.632 13.3171>}]
 dyn_hose2 [{
   shapeType=Capsule
   length=12
   radius=1.6
   shapeName=dyn_hose2_capsule
   center=<5.942 -1.5207 -3.8227>
   rotation=<rot 79.2111 -179.9993 174.0061>
  }]
 dyn_hose4 [{shapeType=Capsule, length=0.0765, radius=1.3132, shapeName=dyn_hose4_capsule, rotation=<rot 90 0 0>}]
 dyn_hose5 [{shapeType=Capsule, length=5, radius=2, shapeName=dyn_hose5_capsule, center=<4.2276 0.5076 0.4587>, rotation=<rot 3.6948 96.0739 -90.0409>}]
 dyn_panel_base [{
   shapeType=Capsule
   length=1.0783
   radius=0.5426
   shapeName=dyn_panel_base_capsule
   center=<0.6754 -0.043 -0.0072>
   rotation=<rot 0 93.2211 -89.3616>
  }]
 dyn_panel_lower [{
   shapeType=Capsule
   length=5.263
   radius=1
   shapeName=dyn_panel_lower_capsule
   center=<4.5941 0.0025 0.0604>
   rotation=<rot 0 89.554 -89.3331>
  }]
 dyn_panel_upper [{
   shapeType=Capsule
   length=8.6503
   radius=3
   shapeName=dyn_panel_upper_capsule
   center=<7.8294 -0.9018 0.0379>
   rotation=<rot 0 89.7006 -89.9786>
  }]
 dyn_strap_L [{
   shapeType=Capsule
   length=2.3835
   radius=1.2363
   shapeName=dyn_strap_L_capsule
   center=<1.5592 -0.0939 -0.0058>
   rotation=<rot 0 85.9119 -89.5526>
  }]
 dyn_strap_R [{
   shapeType=Capsule
   length=2.4472
   radius=1.1201
   shapeName=dyn_strap_R_capsule
   center=<2.352 -0.1889 -0.0014>
   rotation=<rot 0 -90.5026 -89.399>
  }]
 pack [{shapeType=Capsule, length=17.4096, radius=7.5649, shapeName=pack_capsule, center=<-0.2942 -7.4237 -1.1826>, rotation=<rot 0 -90 0>}]
 root [{shapeType=Capsule, length=4, radius=1.6, shapeName=root_capsule}]
}
constraints [
 {bone1Name=dyn_hose1, bone2Name=pack, swing1LimitDegrees=4, swing2LimitDegrees=4, twistLimitDegrees=0.5}
 {bone1Name=dyn_hose2, bone2Name=dyn_hose1, swing1LimitDegrees=8, swing2LimitDegrees=8, twistLimitDegrees=0.5}
 {
  bone1Name=pack
  bone2Name=root
  swing1LimitDegrees=45
  swing1Motion=Locked
  swing2LimitDegrees=45
  swing2Motion=Locked
  twistLimitDegrees=45
  twistMotion=Locked
 }
 {
  bone1Name=dyn_hose4
  bone2Name=pack
  swing1LimitDegrees=45
  swing1Motion=Locked
  swing2LimitDegrees=45
  swing2Motion=Locked
  twistLimitDegrees=45
  twistMotion=Locked
 }
 {bone1Name=dyn_hose5, bone2Name=dyn_hose4, swing1LimitDegrees=10, swing2LimitDegrees=10, twistLimitDegrees=2}
 {bone1Name=dyn_strap_L, bone2Name=pack, swing1LimitDegrees=10, swing2LimitDegrees=0, twistLimitDegrees=0, twistMotion=Locked}
 {bone1Name=dyn_strap_R, bone2Name=pack, swing1LimitDegrees=10, swing2LimitDegrees=0, twistLimitDegrees=45, twistMotion=Locked}
 {
  bone1Name=dyn_arm1
  bone2Name=pack
  swing1LimitDegrees=25
  swing2LimitDegrees=45
  swing2Motion=Locked
  twistLimitDegrees=45
  twistMotion=Locked
 }
 {
  bone1Name=dyn_arm2
  bone2Name=dyn_arm1
  swing1LimitDegrees=45
  swing1Motion=Locked
  swing2LimitDegrees=45
  swing2Motion=Locked
  twistLimitDegrees=45
  twistMotion=Locked
 }
 {
  bone1Name=dyn_arm3
  bone2Name=dyn_arm2
  swing1LimitDegrees=6
  swing2LimitDegrees=45
  swing2Motion=Locked
  twistLimitDegrees=45
  twistMotion=Locked
 }
 {
  bone1Name=dyn_panel_base
  bone2Name=dyn_arm3
  swing1LimitDegrees=2
  swing2LimitDegrees=45
  swing2Motion=Locked
  twistLimitDegrees=45
  twistMotion=Locked
 }
 {
  bone1Name=dyn_panel_upper
  bone2Name=dyn_panel_base
  swing1LimitDegrees=2
  swing2LimitDegrees=45
  swing2Motion=Locked
  twistLimitDegrees=45
  twistMotion=Locked
 }
 {
  bone1Name=dyn_panel_lower
  bone2Name=dyn_panel_upper
  swing1LimitDegrees=20
  swing2LimitDegrees=45
  swing2Motion=Locked
  twistLimitDegrees=45
  twistMotion=Locked
 }
 {
  bone1Name=dyn_cable
  bone2Name=dyn_arm3
  swing1LimitDegrees=45
  swing1Motion=Locked
  swing2LimitDegrees=45
  swing2Motion=Locked
  twistLimitDegrees=45
  twistMotion=Locked
 }
]
